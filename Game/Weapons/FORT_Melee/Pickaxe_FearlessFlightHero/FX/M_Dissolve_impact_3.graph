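# Translucent unlit Niagara sprite material for Fearless Flight Hero pickaxe impact FX, with a UseHexagonShape scalar blending between slice flipbook and hexagon looks.
# A tiled dot texture and panning glitch pattern form a contrast-boosted overlay on the color texture, revealed by dynamic Param2.
# The slice look adds bits-mask breakup and orange tint before the emissive adjustment function.
# Opacity thresholds the chosen shape by the dynamic Dissolve parameter, then applies depth fade and particle alpha.
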Material /Game/Weapons/FORT_Melee/Pickaxe_FearlessFlightHero/FX/M_Dissolve_impact_3
graph {
 properties {bUsedWithNiagaraSprites=true, blendMode=BLEND_Translucent, pixelDepthOffsetMode=PDOM_Legacy, shadingModel=MSM_Unlit, usedShadingModels=Unlit}
 nodes [
  n1 = ParticleColor()
  n2 = TextureSample() {texture=&/T_FFH_Slice_Flipbook}
  n3 = ScalarParameter() {parameterName=UseHexagonShape}
  n4 = OneMinus(None=@n3)
  n5 = TextureSample(UVs=Multiply(A=TextureCoordinate(), B=ScalarParameter() {defaultValue=10, parameterName=DotsTile2})) {samplerType=SAMPLERTYPE_LinearColor, texture=/Game/Weapons/WeaponSkins/Wraps/Textures/CircleFade/T_Wrap_CircleFade_M}
  n6 = Constant2Vector() {g=0.5}
  n7 = Multiply(A=Add(A=TextureCoordinate(), B=LinearInterpolate(B=@n6, Alpha=@n3) {constB=1.3}), B=LinearInterpolate(Alpha=@n3) {constA=0.8, constB=0.4})
  n8 = Constant2Vector() {g=0.5, r=-0.5}
  n9 = LinearInterpolate(Alpha=@n3) {constA=0.5, constB=0}
  n10 = TextureSample(UVs=Panner(Coordinate=@n7, Speed=Multiply(A=@n8, B=@n9)) {speedX=-0.5, speedY=0.5}) {texture=&/T_Pattern_Glitch}
  n11 = TextureSample(UVs=@n7) {texture=&/T_FFH_Color}
  n12 = DynamicParameter() {defaultValue=<rgba 0 1 1 1>, paramNames ["Dissolve(0-1)", Param2, Param3, Param4]}
  n13 = Multiply(A=Add(A=Multiply(A=@n5.G, B=@n10.B), B=@n10.R), B=Step(Y=OneMinus(None=@n11.A), X=Add(A=@n12.Param2) {constB=0.01}))
  n14 = Constant() {r=1}
  n15 = Multiply(
   A=Power(Base=MaterialFunctionCall(In=@n13, Contrast=@n14) {materialFunction=&CheapContrast}.Result) {constExponent=0.5}
   B=LinearInterpolate(Alpha=@n3) {constA=5, constB=5}
  )
  n16 = Add(A=LinearInterpolate(B=@n2.RGB, Alpha=@n4), B=Multiply(A=@n11.RGB, B=Saturate(None=@n15)))
  n17 = TextureSample() {texture=&/T_Mask_Bits}
  n18 = Multiply(
   A=@n16
   B=MaterialFunctionCall(In=OneMinus(None=@n17.G), Contrast=Constant() {r=50}) {materialFunction=&CheapContrast}.Result
  )
  n19 = Multiply(A=Constant3Vector() {constant=<rgba 1 0.0832 0 1>}, B=@n17.G)
  MaterialFunctionCall(Emissive=LinearInterpolate(A=LinearInterpolate(A=@n1.RGB, B=@n16, Alpha=@n16), B=Add(A=@n18, B=@n19), Alpha=@n4)) {materialFunction=&MF_FX_EmissiveAdjustments}
  n20 = TextureSample() {texture=&/T_Hexagon}
  n21 = DynamicParameter() {defaultValue=<rgba 0 1 1 1>, paramNames ["Dissolve(0-1)", Param2, Param3, Param4]}
  n22 = Add(A=@n21."Dissolve(0-1)") {constB=0.01}
  n23 = Saturate(None=Step(Y=OneMinus(None=@n20.B), X=OneMinus(None=@n22)))
  n24 = Multiply(A=@n17.G) {constB=0.1}
  n25 = Step(Y=OneMinus(None=Add(A=@n24, B=@n2.A)), X=OneMinus(None=Add(A=@n21."Dissolve(0-1)") {constB=0.01}))
  n26 = Multiply(A=Multiply(A=Add(A=@n17.R, B=@n17.R), B=Saturate(None=@n25))) {constB=20}
  Multiply(A=DepthFade(Opacity=LinearInterpolate(A=@n23, B=@n26, Alpha=@n4)) {fadeDistanceDefault=10}, B=ParticleColor().A)
 ]
}
