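# Translucent unlit Niagara sprite material for a glitchy, dissolving impact effect on the FearlessFlightHero pickaxe.
# Color combines impact and color textures, particle color and panning glitch/dot patterns, with UseHexagonShape blending between impact and hexagon shapes.
# Opacity combines depth-faded shape, time-animated screen-space glitch and scanline blocks, stepped by dynamic parameter Random Value and scaled by particle alpha.
# Panning digital noise, transformed and scaled by GlitchAmount, forms a vector offset.
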Material /Game/Weapons/FORT_Melee/Pickaxe_FearlessFlightHero/FX/M_Dissolve_impact_4
graph {
 properties {bUsedWithNiagaraSprites=true, blendMode=BLEND_Translucent, pixelDepthOffsetMode=PDOM_Legacy, shadingModel=MSM_Unlit, usedShadingModels=Unlit}
 nodes [
  Constant3Vector() {constant=<rgba 1 0.0832 0 1>}
  n1 = ScalarParameter() {parameterName=UseHexagonShape}
  n2 = OneMinus(None=@n1)
  n3 = TextureSample() {texture=&/T_FFH_MainGlitch_Impact}
  n4 = TextureSample(UVs=Multiply(A=TextureCoordinate(), B=ScalarParameter() {defaultValue=10, parameterName=DotsTile2})) {samplerType=SAMPLERTYPE_LinearColor, texture=/Game/Weapons/WeaponSkins/Wraps/Textures/CircleFade/T_Wrap_CircleFade_M}
  n5 = Constant2Vector() {g=0.5}
  n6 = Multiply(A=Add(A=TextureCoordinate(), B=LinearInterpolate(B=@n5, Alpha=@n1) {constB=1.3}), B=LinearInterpolate(Alpha=@n1) {constA=0.8, constB=0.4})
  n7 = Constant2Vector() {g=0.5, r=-0.5}
  n8 = LinearInterpolate(Alpha=@n1) {constA=0.5, constB=0}
  n9 = TextureSample(UVs=Panner(Coordinate=@n6, Speed=Multiply(A=@n7, B=@n8)) {speedX=-0.5, speedY=0.5}) {texture=&/T_Pattern_Glitch}
  n10 = TextureSample(UVs=@n6) {texture=&/T_FFH_Color}
  n11 = DynamicParameter() {defaultValue=<rgba 0 7.5 1 1>, paramNames [Random Value, Speed, GlitchAmount, Chroma]}
  n12 = Multiply(A=Add(A=Multiply(A=@n4.G, B=@n9.B), B=@n9.R), B=Step(Y=OneMinus(None=@n10.A), X=Add(A=@n11.Speed) {constB=0.01}))
  n13 = Constant() {r=1}
  n14 = Multiply(
   A=Power(Base=MaterialFunctionCall(In=@n12, Contrast=@n13) {materialFunction=&CheapContrast}.Result) {constExponent=0.5}
   B=LinearInterpolate(Alpha=@n1) {constA=5, constB=5}
  )
  n15 = Add(A=LinearInterpolate(A=@n3.RGB, B=@n3.RGB, Alpha=@n2), B=Multiply(A=@n10.RGB, B=Saturate(None=@n14)))
  n16 = TextureSample() {texture=&T_FFH_MainGlitch_Impact}
  n17 = MaterialFunctionCall(In=OneMinus(None=@n16.R), Contrast=Constant() {r=50}) {materialFunction=&CheapContrast}
  LinearInterpolate(
   A=LinearInterpolate(A=ParticleColor().RGB, B=@n15, Alpha=@n15)
   B=Add(A=Multiply(A=@n15, B=@n17.Result), B=Multiply(B=@n3.RGB)) {constB=0}
   Alpha=@n2
  )
  n18 = DynamicParameter() {defaultValue=<rgba 0 7.5 1 1>, paramNames [Random Value, Speed, GlitchAmount, Chroma]}
  n19 = MaterialFunctionCall(UVs=Multiply(A=ScreenPosition().ViewportUV, B=LinearInterpolate(Alpha=@n18.Random Value) {constA=5, constB=10})) {materialFunction=&ScaleUVsByCenter}
  n20 = OneMinus(None=TextureSample(UVs=@n19.UVs) {samplerSource=SSM_Wrap_WorldGroupSettings, texture=&/T_Glitch_Pattern_FFH}.R)
  n21 = Time()
  n22 = Add(A=@n21, B=@n18.Random Value)
  n23 = Multiply(A=Add(A=Cosine(None=@n22))) {constB=0.5}
  n24 = Add(A=@n21, B=@n18.Random Value)
  n25 = Multiply(A=Add(A=Sine(None=@n24))) {constB=0.5}
  n26 = Multiply(A=@n25, B=@n23)
  n27 = TextureSample() {samplerSource=SSM_Clamp_WorldGroupSettings, texture=&T_Glitch_Pattern_FFH}
  n28 = Add(A=@n25) {constB=-0.5}
  n29 = LinearInterpolate(A=LinearInterpolate(A=Step(Y=@n20, X=@n23), Alpha=@n26), B=Step(Y=OneMinus(None=@n27.G), X=Multiply(A=@n28) {constB=1.5}), Alpha=@n26)
  n30 = DynamicParameter() {defaultValue=<rgba 0 7.5 1 1>, paramNames [Random Value, Speed, GlitchAmount, Chroma]}
  n31 = Step(
   Y=OneMinus(None=TextureSample() {texture=&/T_Hexagon}.B)
   X=OneMinus(None=Add(A=@n30.Random Value) {constB=0.01})
  )
  n32 = Add(A=Multiply(A=@n16.R) {constB=0.1}, B=@n3.A)
  n33 = Add(A=@n30.Random Value) {constB=0.01}
  n34 = Saturate(None=Step(Y=OneMinus(None=@n32), X=OneMinus(None=@n33)))
  n35 = Multiply(A=Add(A=@n16.R, B=@n16.R), B=@n34)
  n36 = DepthFade(Opacity=LinearInterpolate(A=Saturate(None=@n31), B=Multiply(A=@n35) {constB=20}, Alpha=@n2)) {fadeDistanceDefault=10}
  n37 = Multiply(A=Add(A=@n29, B=@n36), B=Add(A=@n36, B=Multiply(A=@n34, B=@n29)))
  n38 = Multiply(A=@n27.G, B=@n29)
  n39 = Add(A=DynamicParameter() {defaultValue=<rgba 0 7.5 1 1>, paramNames [Random Value, Speed, GlitchAmount, Chroma]}.Random Value) {constB=0.01}
  n40 = DynamicParameter() {parameterIndex=1, defaultValue=<rgba 1 0 0 0>, paramNames [EdgeSpeed, BiasScale, Gradient Position, TriggerWiggle]}
  n41 = Constant()
  n42 = StaticSwitchParameter(
   True=MaterialFunctionCall!(
    WPO Intensity A
    WPO Intensity B
    Time
    Bias=StaticSwitchParameter(True=@n40.TriggerWiggle, False=@n41) {defaultValue=true, parameterName=Use Param4 for Wiggle}
   ) {materialFunction=/Game/Effects/Fort_Effects/Effects/Atlas/Materials/MF_Atlas_HologramWiggle_01}.Intense 0-1
   False=Constant()
  ) {defaultValue=true, parameterName=Use Blocks?}
  n43 = Multiply(
   A=Step(Y=OneMinus(None=@n38), X=OneMinus(None=@n39))
   B=MaterialFunctionCall!(
    Number of Scanlines=ScalarParameter() {defaultValue=150, parameterName=Scanline Amount}
    Scanline Speed
    "Scanlines Thickness (Bias)"=ScalarParameter() {defaultValue=-0.15, parameterName=ScanlineSize}
    Scanlines Dark Value=ScalarParameter() {defaultValue=0.5, parameterName=Scanlines Dark}
    Scanlines Light Value=ScalarParameter() {defaultValue=0.6, parameterName=Scanlines Light}
    Block Opacity=Multiply(A=@n42) {constB=0.4}
    UVs
    Time
    Block Repeat
   ) {materialFunction=/Game/Effects/Fort_Effects/Effects/Atlas/Materials/MF_Atlas_Scanlines_01}.Result
  )
  Multiply(A=Saturate(None=Add(A=@n37, B=@n43)), B=ParticleColor().A)
  n44 = DynamicParameter() {defaultValue=<rgba 0 7.5 1 1>, paramNames [Random Value, Speed, GlitchAmount, Chroma]}
  n45 = Panner(Coordinate=Add(A=TextureCoordinate() {uTiling=0.01, vTiling=0.01}, B=@n44.Random Value), Time=Multiply(A=@n44.Speed, B=Time())) {speedX=0.1, speedY=0.1}
  Multiply(
   A=Transform(None=TextureSample(UVs=@n45) {samplerType=SAMPLERTYPE_LinearColor, texture=/Game/Global/Textures/Default/Noise/T_Atlas_DigitalNoise_01}.RGB)
   B=ScalarParameter() {defaultValue=5, parameterName=GlitchAmount}
  )
  Reroute(None=@n16.R)
  Reroute(None=@n3.RGB)
  Saturate(None=MaterialFunctionCall(Emissive=@n15) {materialFunction=&MF_FX_EmissiveAdjustments}.Result)
  TextureCoordinate()
  TextureSample(UVs=@n19.UVs) {samplerType=SAMPLERTYPE_LinearColor, texture=&T_Wrap_CircleFade_M}
 ]
}
