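# Unlit translucent Niagara sprite material for the toon big-shape impact effect of the Fearless Flight Hero pickaxe.
# Emissive multiplies particle color, shape alpha and a scaled-UV color texture, then passes through an emissive-adjustment function.
# Opacity carves the shape alpha with its red channel plus tiled circle-fade dots, offset by a dynamic-parameter dissolve.
# A high-contrast step hardens the dissolve edge before depth fade and particle alpha apply.
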Material /Game/Weapons/FORT_Melee/Pickaxe_FearlessFlightHero/FX/M_FearlessFlightHero_BigShape_Toon_Impact
graph {
 properties {bUsedWithNiagaraSprites=true, blendMode=BLEND_Translucent, pixelDepthOffsetMode=PDOM_Legacy, shadingModel=MSM_Unlit, usedShadingModels=Unlit}
 nodes [
  n1 = ParticleColor()
  n2 = TextureSample() {texture=&/T_Impact_BigShape_Toon}
  n3 = TextureCoordinate() {uTiling=0.2, vTiling=0.2}
  MaterialFunctionCall(Emissive=Multiply(A=Multiply(A=@n1.RGB, B=@n2.A), B=TextureSample(UVs=@n3) {texture=&/T_FFH_Color}.RGB)) {materialFunction=&MF_FX_EmissiveAdjustments}
  n4 = TextureSample(UVs=Multiply(A=TextureCoordinate(), B=ScalarParameter() {defaultValue=2.5, parameterName=DotsTile2})) {samplerType=SAMPLERTYPE_LinearColor, texture=/Game/Weapons/WeaponSkins/Wraps/Textures/CircleFade/T_Wrap_CircleFade_M}
  n5 = Multiply(A=@n4.G) {constB=0.2}
  n6 = Subtract(
   A=Saturate(None=Add(A=@n2.R, B=@n5))
   B=DynamicParameter() {defaultValue=<rgba -0.7 1 1 1>, paramNames ["Dissolve(-1/1)", Param2, Param3, Param4]}."Dissolve(-1/1)"
  )
  n7 = MaterialFunctionCall(In=Subtract(A=@n2.A, B=Saturate(None=@n6)), Contrast=Constant() {r=5}) {materialFunction=&CheapContrast}
  Multiply(A=@n1.A, B=DepthFade(Opacity=Saturate(None=@n7.Result)) {fadeDistanceDefault=5})
  OneMinus(None=@n4.G)
  TextureSample() {samplerType=SAMPLERTYPE_Masks, texture=/Game/Weapons/WeaponSkins/Wraps/Textures/ComicBubbles/T_Wrap_ComicBubble_H}
 ]
}
